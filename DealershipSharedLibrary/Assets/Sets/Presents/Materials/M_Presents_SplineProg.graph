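# Masked unlit Niagara material for a cable with colored lights in the Presents set, with a progressive reveal and dissolve.
# Vertex color blue separates mesh parts, one clipped by a particle-alpha-driven mask, the other scaled toward UV-baked pivots through world position offset, which also adds sine sway.
# Emissive combines shaded cable color, two-color lights and a glowing dissolve edge from screen-space noise, with dynamic parameters driving sway, fade and light intensity.
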Material /DealershipSharedLibrary/Assets/Sets/Presents/Materials/M_Presents_SplineProg
graph {
 properties {
  bAutomaticallySetUsageInEditor=false
  bUsedWithNiagaraMeshParticles=true
  bUsedWithNiagaraRibbons=true
  bUsedWithNiagaraSprites=true
  blendMode=BLEND_Masked
  shadingModel=MSM_Unlit
  usedShadingModels=Unlit
 }
 nodes [
  MaterialFunctionCall(Emissive=NamedRerouteUsage()) {materialFunction=&MF_FX_EmissiveAdjustments}
  n1 = Subtract(
   A=LinearInterpolate(
    A=ScalarParameter() {defaultValue=-64, parameterName=BegMask}
    B=ScalarParameter() {defaultValue=4096, parameterName=MaskScale}
    Alpha=ParticleColor().A
   )
   B=NamedRerouteUsage()
  )
  NamedRerouteDeclaration(None=@n1) {name=Distance Mask, nodeColor=<rgba 1 0 0.5859 1>}
  n2 = Saturate(None=@n1)
  n3 = Constant() {r=1}
  n4 = Ceil(None=VertexColor().B)
  NamedRerouteDeclaration(None=Multiply(A=NamedRerouteUsage(), B=LinearInterpolate(A=@n2, B=@n3, Alpha=@n4))) {name=Opacity Mask, nodeColor=<rgba 0 0.9062 1 1>}
  n5 = Ceil(None=VertexColor().B)
  n6 = Multiply(A=NamedRerouteUsage(), B=Constant() {r=0.001})
  n7 = TextureCoordinate() {coordinateIndex=2}
  n8 = Add(
   A=AppendVector(A=TextureCoordinate() {coordinateIndex=1}, B=ComponentMask(None=@n7) {r=true})
   B=Constant3Vector() {constant=<rgba 0 -1 0 0>}
  )
  n9 = WorldPosition()
  n10 = Multiply(
   A=OneMinus(None=Saturate(None=@n6))
   B=Subtract(A=TransformPosition(None=@n8) {transformSourceType=TRANSFORMPOSSOURCE_Instance, transformType=TRANSFORMPOSSOURCE_World}, B=@n9.XYZ)
  )
  n11 = NamedRerouteUsage()
  n12 = Constant() {r=128}
  n13 = Saturate(None=Multiply(A=Subtract(A=@n11, B=@n12), B=Constant() {r=0.001}))
  n14 = DynamicParameter() {defaultValue=<rgba 100 0.0002 1 500>, paramNames [Intensity,Tile,Speed,FadeOut]}
  n15 = Add(A=Multiply(A=NamedRerouteUsage(), B=@n14.Tile), B=Multiply(A=@n14.Speed, B=Time()))
  n16 = Sine(None=@n15)
  n17 = Multiply(
   A=AppendVector(A=AppendVector(A=@n16, B=@n16), B=Cosine(None=@n15))
   B=AppendVector(A=AppendVector(A=@n14.Intensity, B=@n14.Intensity), B=@n14.Intensity)
  )
  NamedRerouteDeclaration(None=Add(A=Multiply(A=@n5, B=@n10), B=Multiply(A=@n13, B=@n17))) {name=WPO, nodeColor=<rgba 0 0.1094 1 1>}
  n18 = ComponentMask(None=TextureCoordinate() {coordinateIndex=2}) {g=true}
  n19 = ScalarParameter() {defaultValue=0.5, parameterName=LightRange}
  n20 = LinearInterpolate(
   A=VectorParameter() {parameterName=LightColor01, defaultValue=<rgba 0.03 1 0.03 1>}.RGBA
   B=VectorParameter() {parameterName=LightColor02, defaultValue=<rgba 2 0 0.1 1>}.RGBA
   Alpha=Ceil(None=Subtract(A=@n18, B=@n19))
  )
  n21 = Constant() {r=6}
  NamedRerouteDeclaration(None=Multiply(
    A=Multiply(A=@n20, B=@n21)
    B=DynamicParameter() {parameterIndex=1, defaultValue=<rgba 1 0 0 0>, paramNames [Lights_Intensity, 1, 2, 3]}.Lights_Intensity
   )) {name=Lights, nodeColor=<rgba 0 0.3437 1 1>}
  n22 = Constant3Vector() {constant=<rgba 0 0 1 0>}
  n23 = DotProduct(A=Normalize(VectorInput=Transform(None=@n22)), B=Normalize(VectorInput=Constant3Vector() {constant=<rgba 0 0 1 0>}))
  n24 = Constant() {r=0.7}
  n25 = Power(Base=Multiply(A=Add(A=@n23, B=@n24), B=@n24), Exp=Constant() {r=3})
  n26 = Constant() {r=4}
  n27 = Multiply(
   A=VectorParameter() {parameterName=CableColor, defaultValue=<rgba 0.05 0.1 0.05 1>}.RGBA
   B=Clamp(None=Multiply(A=@n25, B=@n26)) {maxDefault=2.5, minDefault=1}
  )
  n28 = Fresnel() {exponent=3}
  NamedRerouteDeclaration(None=Add(A=@n27, B=Multiply(A=@n27, B=@n28))) {name=String, nodeColor=<rgba 0.6641 1 0 1>}
  n29 = ScreenPosition()
  n30 = ParticleColor()
  n31 = Multiply(A=Multiply(A=ComponentMask(None=@n29.ViewportUV) {g=true, r=true}, B=ViewSize()), B=@n30.G)
  n32 = Add(
   A=Multiply(A=TextureSampleParameter2D(UVs=@n31) {parameterName=Tex, texture=/DealershipSharedLibrary/FX/Textures/Noise/T_VCVFX_BlobNoise}.G, B=@n30.R)
   B=NamedRerouteUsage()
  )
  n33 = DynamicParameter() {defaultValue=<rgba 100 0.0002 1 500>, paramNames [Intensity,Tile,Speed,FadeOut]}
  NamedRerouteDeclaration(None=Saturate(None=Subtract(A=@n32, B=@n33.FadeOut))) {name=Dissolve, nodeColor=<rgba 1 0 0.9844 1>}
  NamedRerouteDeclaration(None=Add(A=@n32, B=ScalarParameter() {defaultValue=-200, group=Dissolve, parameterName=Dissolve Edge, sortPriority=45})) {name=Dissolve Edge, nodeColor=<rgba 0 1 0.6953 1>}
  NamedRerouteDeclaration(None=Distance(
    A=@n9.XYZ
    B=MaterialFunctionCall() {materialFunction=&ComponentPivotLocation}.Result
   )) {name=Distance, nodeColor=<rgba 1 0.375 0 1>}
  n34 = Subtract(A=DynamicParameter() {defaultValue=<rgba 100 0.0002 1 500>, paramNames [Intensity,Tile,Speed,FadeOut]}.FadeOut, B=NamedRerouteUsage())
  n35 = ScalarParameter() {defaultValue=12, parameterName=Int_Dissolve Edge Color}
  n36 = LinearInterpolate(
   A=VectorParameter() {parameterName=Dissolve Edge Color01, defaultValue=<rgba 0.03 1 0.03 1>}.RGBA
   B=VectorParameter() {parameterName=Dissolve Edge Color02, defaultValue=<rgba 1 0.03 0.03 1>}.RGBA
   Alpha=ParticleColor().B
  )
  NamedRerouteDeclaration(None=Multiply(A=Saturate(None=@n34), B=Multiply(A=@n35, B=@n36))) {name=Dissolve Edge Color, nodeColor=<rgba 0.0547 0 1 1>}
  n37 = NamedRerouteUsage()
  n38 = NamedRerouteUsage()
  n39 = Ceil(None=VertexColor().B)
  NamedRerouteDeclaration(None=Add(A=LinearInterpolate(A=@n37, B=@n38, Alpha=@n39), B=NamedRerouteUsage())) {name=Emissive Color, nodeColor=<rgba 0.2656 1 0 1>}
  NamedRerouteUsage()
  NamedRerouteUsage()
 ]
}
